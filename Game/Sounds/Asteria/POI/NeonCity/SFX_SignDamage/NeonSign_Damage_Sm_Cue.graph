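# Sound cue for damage to small neon signs in the NeonCity POI, reusing small speaker impact waves.
# An equally weighted Random node picks one of five waves, then feeds two chained Modulators that vary pitch and volume.
# The cue raises overall pitch, plays through the impacts sound class and uses neon sign damage attenuation.
# Overridden concurrency caps voices per owner, blocks new plays at the limit and sets a retrigger time.
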SoundCue /Game/Sounds/Asteria/POI/NeonCity/SFX_SignDamage/NeonSign_Damage_Sm_Cue
root=Modulator_2
nodes [
 {id=Modulator_1, children [Random_2]}
 {id=Modulator_2, children [Modulator_1], properties {pitchMax=1.1, pitchMin=0.9, volumeMin=0.7}}
 {id=Random_2, children [WavePlayer_9, WavePlayer_10, WavePlayer_11, WavePlayer_12, WavePlayer_13], properties {weights [1,1,1,1,1]}}
 {id=WavePlayer_10, properties {soundWaveAssetPtr=/Game/Sounds/Artemis/Biomes/HollowBaseParty/Speaker_SFX/wav/Speaker_Sm_Impact_02}}
 {id=WavePlayer_11, properties {soundWaveAssetPtr=/Game/Sounds/Artemis/Biomes/HollowBaseParty/Speaker_SFX/wav/Speaker_Sm_Impact_03}}
 {id=WavePlayer_12, properties {soundWaveAssetPtr=/Game/Sounds/Artemis/Biomes/HollowBaseParty/Speaker_SFX/wav/Speaker_Sm_Impact_04}}
 {id=WavePlayer_13, properties {soundWaveAssetPtr=/Game/Sounds/Artemis/Biomes/HollowBaseParty/Speaker_SFX/wav/Speaker_Sm_Impact_05}}
 {id=WavePlayer_9, properties {soundWaveAssetPtr=/Game/Sounds/Artemis/Biomes/HollowBaseParty/Speaker_SFX/wav/Speaker_Sm_Impact_01}}
]
properties {
 attenuationSettings=&/Att_NeonSign_Damage
 bOverrideConcurrency=true
 maxDistance=3450
 pitchMultiplier=1.3
 soundClassObject=&Fort_Impacts
 soundSubmixObject=&Fortnite_SFX
 volumeMultiplier=0.8
 attenuationOverrides {falloffDistance=4650, attenuationShapeExtents {x=800}}
 concurrencyOverrides {bLimitToOwner=true, maxCount=2, resolutionRule=PreventNew, retriggerTime=1.2, volumeScale=0.3}
}
