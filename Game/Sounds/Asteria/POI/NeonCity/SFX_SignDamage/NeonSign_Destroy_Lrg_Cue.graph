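# Sound cue for the destruction of a large neon sign at the Neon City POI, routed to the Fort_Impacts sound class.
# A Random node picks one of six equally weighted wave variations and feeds it through a Modulator for slight volume variation.
# Inline attenuation overrides add distance low-pass filtering, listener focus and occlusion.
# Concurrency overrides allow a single instance, blocking new plays with a retrigger time to limit stacking.
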SoundCue /Game/Sounds/Asteria/POI/NeonCity/SFX_SignDamage/NeonSign_Destroy_Lrg_Cue
root=Modulator_0
nodes [
 {id=Modulator_0, children [Random_1], properties {volumeMin=0.8}}
 {id=Random_1, children [WavePlayer_2,WavePlayer_3,WavePlayer_4,WavePlayer_5,WavePlayer_6,WavePlayer_7], properties {weights [1,1,1,1,1,1]}}
 {id=WavePlayer_2, properties {soundWaveAssetPtr=/Game/Sounds/Asteria/POI/NeonCity/SFX_SignDamage/wav/NeonSign_Destroy_Lrg_01}}
 {id=WavePlayer_3, properties {soundWaveAssetPtr=/Game/Sounds/Asteria/POI/NeonCity/SFX_SignDamage/wav/NeonSign_Destroy_Lrg_02}}
 {id=WavePlayer_4, properties {soundWaveAssetPtr=/Game/Sounds/Asteria/POI/NeonCity/SFX_SignDamage/wav/NeonSign_Destroy_Lrg_03}}
 {id=WavePlayer_5, properties {soundWaveAssetPtr=/Game/Sounds/Asteria/POI/NeonCity/SFX_SignDamage/wav/NeonSign_Destroy_Lrg_04}}
 {id=WavePlayer_6, properties {soundWaveAssetPtr=/Game/Sounds/Asteria/POI/NeonCity/SFX_SignDamage/wav/NeonSign_Destroy_Lrg_05}}
 {id=WavePlayer_7, properties {soundWaveAssetPtr=/Game/Sounds/Asteria/POI/NeonCity/SFX_SignDamage/wav/NeonSign_Destroy_Lrg_06}}
]
properties {
 attenuationSettings=&/Att_NeonSign_Damage
 bOverrideAttenuation=true
 bOverrideConcurrency=true
 maxDistance=10500
 soundClassObject=&Fort_Impacts
 volumeMultiplier=0.84
 attenuationOverrides {
  bAttenuateWithLPF=true
  bEnableListenerFocus=true
  bEnableOcclusion=true
  falloffDistance=9000
  lPFFrequencyAtMax=900
  lPFFrequencyAtMin=8500
  lPFRadiusMax=3500
  lPFRadiusMin=850
  nonFocusVolumeAttenuation=0.8
  nonSpatializedRadiusStart=30
  occlusionLowPassFilterFrequency=2000
  occlusionTraceChannel=ECC_GameTraceChannel11
  occlusionVolumeAttenuation=0.6
  stereoSpread=3000
  attenuationShapeExtents {x=1500}
 }
 concurrencyOverrides {bVolumeScaleCanRelease=true, maxCount=1, resolutionRule=PreventNew, retriggerTime=1.2, voiceStealReleaseTime=0.3, volumeScale=0.4}
}
