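# Sound cue for the destruction of a small neon sign in the Asteria NeonCity POI.
# A random node picks between two equally weighted small-speaker death waves borrowed from Artemis audio, then feeds a modulator.
# The cue raises pitch and lowers volume, using the neon sign damage attenuation asset and the impacts sound class.
# Its concurrency override caps simultaneous plays, prevents new ones at the limit and scales down overlapping volume.
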SoundCue /Game/Sounds/Asteria/POI/NeonCity/SFX_SignDamage/NeonSign_Destroy_Sm_Cue
root=Modulator_1
nodes [
 {id=Modulator_1, children [Random_1]}
 {id=Random_1, children [WavePlayer_2,WavePlayer_3], properties {weights [1,1]}}
 {id=WavePlayer_2, properties {soundWaveAssetPtr=/Game/Sounds/Artemis/Biomes/HollowBaseParty/Speaker_SFX/wav/Speaker_Sm_Death_01}}
 {id=WavePlayer_3, properties {soundWaveAssetPtr=/Game/Sounds/Artemis/Biomes/HollowBaseParty/Speaker_SFX/wav/Speaker_Sm_Death_02}}
]
properties {
 attenuationSettings=&/Att_NeonSign_Damage
 bOverrideConcurrency=true
 maxDistance=3450
 pitchMultiplier=1.5
 soundClassObject=&Fort_Impacts
 volumeMultiplier=0.6
 attenuationOverrides {bAttenuateWithLPF=true, falloffDistance=4650, attenuationShapeExtents {x=800}}
 concurrencyOverrides {bVolumeScaleCanRelease=true, maxCount=2, resolutionRule=PreventNew, retriggerTime=1, voiceStealReleaseTime=0.3, volumeScale=0.5}
}
